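# Reticle extension widget for the Rocket Ram weapon, showing a crosshair, charge-up progress image and cooldown timer.
# On construct it binds the weapon's ram-state delegate and starts async gameplay event and tag listeners on the owning pawn, released on destruct.
# Wind-up events toggle a looping timer driving the charge material's Progress, using event magnitude as duration.
# Crosshair shows only in certain ram states without cooldown or trajectory display.
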WidgetBlueprint /RocketRamGameplayUI/Widgets/WBP_RocketRam_Reticle
graphs [
 EventGraph {nodes [
   n1 = OnRamStateChanged!() -> (OutputDelegate, then, RamState:Enum_RocketRamState Enum=NewEnumerator0) {is=title K2Node_CustomEvent}
   > Set CurrentState(CurrentState:Enum_RocketRamState Enum=@n1.RamState) -> (Output_Get:Enum_RocketRamState Enum=NewEnumerator0)
   > RefreshReticleVisibility()
   Event Destruct()
   > Parent: Destruct()
   > EndAction(self:Ability Async Object Reference=Get WaitForWindUpStartEventHandle() -> (WaitForWindUpStartEventHandle:Ability Async Wait Gameplay Event Object Reference))
   > EndAction(self=Get WaitForWindUpEndEventHandle() -> (WaitForWindUpEndEventHandle:Ability Async Wait Gameplay Event Object Reference))
   > EndAction(self=Get WaitForShowTrajectoryAbilityAdded() -> (WaitForShowTrajectoryAbilityAdded:Ability Async Wait Gameplay Tag Added Object Reference))
   > EndAction(self=Get WaitForShowTrajectoryAbilityRemoved() -> (WaitForShowTrajectoryAbilityRemoved:Ability Async Wait Gameplay Tag Removed Object Reference))
   n2 = Get RocketRamWeapon() -> (RocketRamWeapon:B Weapon Rocket Ram Object Reference)
   > Is Valid(InputObject=@n2)
   > Unbind Event from On Ram State Changed Event(self:B Weapon Rocket Ram Object Reference=@n2, Delegate=Create Event()) {is=title K2Node_RemoveDelegate}
   Event Construct()
   > Parent: Construct()
   > SetWindUpProgress!(execute, self, Value:float=0) -> (then)
   > SetVisibility(self=Get Image_ChargeUp() -> (Image_ChargeUp:Image Object Reference), InVisibility=Collapsed)
   > n3 = Cast To FortPawn!(execute, Object=GetOwningPlayerPawn!(self:User Widget Object Reference) -> (ReturnValue:Pawn Object Reference)) -> (then, CastFailed, AsFort Pawn:Fort Pawn Object Reference) {is=title K2Node_DynamicCast}
   > n4 = Set FortPawn(FortPawn:Fort Pawn Object Reference=@n3) -> (Output_Get:Fort Pawn Object Reference)
   > n5 = WaitGameplayEventToActor!(
    execute:Exec
    TargetActor:Actor Object Reference=@n4
    EventTag:Gameplay Tag Structure=Get GE_WindUpStart() -> (GE_WindUpStart:Gameplay Tag Structure)
    OnlyTriggerOnce:bool=false
    OnlyMatchExact:bool=true
   ) -> (then:Exec, AsyncTaskProxy:Ability Async Wait Gameplay Event Object Reference, EventReceived:Exec, Payload:Gameplay Event Data Structure) {is=title K2Node_AsyncAction}
   > Set WaitForWindUpStartEventHandle(WaitForWindUpStartEventHandle:Ability Async Wait Gameplay Event Object Reference=@n5) -> (Output_Get:Ability Async Wait Gameplay Event Object Reference)
   > n6 = WaitGameplayEventToActor(TargetActor=@n4, EventTag=Get GE_WindUpEnd() -> (GE_WindUpEnd:Gameplay Tag Structure), OnlyTriggerOnce=false, OnlyMatchExact=true)
   > Set WaitForWindUpEndEventHandle(WaitForWindUpEndEventHandle:Ability Async Wait Gameplay Event Object Reference=@n6) -> (Output_Get:Ability Async Wait Gameplay Event Object Reference)
   > n7 = WaitGameplayTagAddToActor!(
    execute:Exec
    TargetActor:Actor Object Reference=@n4
    Tag:Gameplay Tag Structure=Get GA_ShowTrajectory() -> (GA_ShowTrajectory:Gameplay Tag Structure)
    OnlyTriggerOnce:bool=false
   ) -> (then:Exec, AsyncTaskProxy:Ability Async Wait Gameplay Tag Added Object Reference, Added:Exec) {is=title K2Node_AsyncAction}
   > Set WaitForShowTrajectoryAbilityAdded(WaitForShowTrajectoryAbilityAdded:Ability Async Wait Gameplay Tag Added Object Reference=@n7) -> (Output_Get:Ability Async Wait Gameplay Tag Added Object Reference)
   > n8 = WaitGameplayTagRemoveFromActor!(execute:Exec, TargetActor:Actor Object Reference=@n4, Tag:Gameplay Tag Structure=Get GA_ShowTrajectory(), OnlyTriggerOnce:bool=false) -> (then:Exec, AsyncTaskProxy:Ability Async Wait Gameplay Tag Removed Object Reference, Removed:Exec) {is=title K2Node_AsyncAction}
   > Set WaitForShowTrajectoryAbilityRemoved(WaitForShowTrajectoryAbilityRemoved:Ability Async Wait Gameplay Tag Removed Object Reference=@n8) -> (Output_Get:Ability Async Wait Gameplay Tag Removed Object Reference)
   > n9 = Cast To B_Weapon_RocketRam!(
    execute
    Object=GetEquippedWeapon!(self:Fort Weapon Reticle Extension Widget Base Object Reference) -> (ReturnValue:Fort Weapon Object Reference)
   ) -> (then, CastFailed, AsB Weapon Rocket Ram:B Weapon Rocket Ram Object Reference) {is=title K2Node_DynamicCast}
   > n10 = Set RocketRamWeapon(RocketRamWeapon:B Weapon Rocket Ram Object Reference=@n9) -> (Output_Get:B Weapon Rocket Ram Object Reference)
   > Bind Event to On Ram State Changed Event(self:B Weapon Rocket Ram Object Reference=@n10, Delegate=Create Event()) {is=title K2Node_AddDelegate}
   > RefreshReticleVisibility()
   SetShowTrajectoryState!(execute=@n8.Removed, self, IsShowing:bool=false) -> (then)
   SetShowTrajectoryState(execute=@n7.Added, IsShowing=true)
   SetWindUpState!(execute=@n6.EventReceived, self, IsActive:bool=false, GameplayEventData:"Gameplay Event Data Structure (by ref)"=@n6.Payload) -> (then)
   SetWindUpState(execute=@n5.EventReceived, IsActive=true, GameplayEventData=@n5.Payload)
   Event OnInitialized()
   > Bind Event to On Cooldown State Changed(
    self:WBP Reticle Cooldown Timer Object Reference=Get WBP_Reticle_CooldownTimer() -> (WBP_Reticle_CooldownTimer:WBP Reticle Cooldown Timer Object Reference)
    Delegate=Create Event()
   ) {is=title K2Node_AddDelegate}
  ]}
 OnCooldownStateChanged {nodes [
   n1 = OnCooldownStateChanged!() -> (then, IsCooldownActive:bool=~) {is=title K2Node_FunctionEntry}
   > Set Is Cooldown Active(Is Cooldown Active:bool=@n1) -> (Output_Get:bool)
   > RefreshReticleVisibility()
  ]}
 RefreshReticleVisibility {nodes [
   RefreshReticleVisibility() {is=title K2Node_FunctionEntry}
   > n1 = Get!(execute:Exec) -> (then:Exec, else:Exec, RocketRamWeapon:B Weapon Rocket Ram Object Reference)
   > n2 = Get!(execute) -> (then, else, Image_Crosshair:Image Object Reference)
   n3 = NOT Boolean(A=Get IsTrajectoryShowing() -> (IsTrajectoryShowing:bool))
   n4 = NOT Boolean(A=Get Is Cooldown Active() -> (Is Cooldown Active:bool))
   n5 = Select!(
    NewEnumerator0:bool=true
    NewEnumerator6:bool=true
    NewEnumerator1:bool=false
    NewEnumerator2:bool=false
    NewEnumerator5:bool=false
    NewEnumerator3:bool=false
    NewEnumerator7:bool=false
    NewEnumerator4:bool=true
    Index:Enum_RocketRamState Enum=Get RamState!(self:B Weapon Rocket Ram Object Reference=@n1) -> (RamState:Enum_RocketRamState Enum=NewEnumerator0)
   ) -> (ReturnValue:bool)
   > SetVisibility(
    self=@n2
    InVisibility=Select!(
     Option 0:ESlateVisibility Enum=Collapsed
     Option 1:ESlateVisibility Enum=SelfHitTestInvisible
     Index:bool=AND Boolean!(A=@n3, B=@n4, C:bool=@n5) -> (ReturnValue)
    ) -> (ReturnValue:ESlateVisibility Enum)
   )
  ]}
 SetShowTrajectoryState {nodes [
   n1 = SetShowTrajectoryState!() -> (then, IsShowing:bool=~) {is=title K2Node_FunctionEntry}
   > Set IsTrajectoryShowing(IsTrajectoryShowing:bool=@n1) -> (Output_Get:bool)
   > RefreshReticleVisibility() {is=title K2Node_CallFunction}
  ]}
 SetWindUpProgress {nodes [
   n1 = SetWindUpProgress!() -> (then, Value:float=~) {is=title K2Node_FunctionEntry}
   > n2 = GetDynamicMaterial(self=Get Image_ChargeUp())
   > SetScalarParameterValue(self=@n2, ParameterName=Progress, Value=@n1)
  ]}
 SetWindUpState {nodes [
   n1 = SetWindUpState!() -> (then, IsActive:bool=~, GameplayEventData:"Gameplay Event Data Structure (by ref)") {is=title K2Node_FunctionEntry}
   > n2 = Set IsWindUpActive(IsWindUpActive:bool=@n1) -> (Output_Get:bool)
   > RefreshReticleVisibility()
   > n3 = Branch(Condition=@n2)
   > Set WindUpStartTime(WindUpStartTime:double=GetGameTimeInSeconds()) -> (Output_Get:double)
   > Set WindUpDuration(WindUpDuration:float=Break Gameplay Event Data!(GameplayEventData:"Gameplay Event Data Structure (by ref)"=@n1.GameplayEventData) -> (
     EventTag:Gameplay Tag Structure
     Instigator:Actor Object Reference
     Target:Actor Object Reference
     OptionalObject:Object Reference
     OptionalObject2:Object Reference
     ContextHandle:Gameplay Effect Context Handle Structure
     InstigatorTags:Gameplay Tag Container Structure
     TargetTags:Gameplay Tag Container Structure
     EventMagnitude:float=~
     TargetData:Gameplay Ability Target Data Handle Structure
     InstancedEventData:Instanced Struct Structure
    ) {is=title K2Node_BreakStruct}.EventMagnitude) -> (Output_Get:float)
   > n4 = Set Timer by Event(Delegate=Create Event(), Time=0.01, bLooping=true)
   > Set WindUpUpdateTimerHandle(WindUpUpdateTimerHandle:Timer Handle Structure=@n4) -> (Output_Get:Timer Handle Structure)
   > SetVisibility(self=Get Image_ChargeUp(), InVisibility=SelfHitTestInvisible)
   Clear and Invalidate Timer by Handle(execute=@n3.else, Handle=Get WindUpUpdateTimerHandle() -> (WindUpUpdateTimerHandle:Timer Handle Structure))
   > SetWindUpProgress(Value=0) {is=title K2Node_CallFunction}
   > SetVisibility(self=Get Image_ChargeUp(), InVisibility=Collapsed)
  ]}
 UpdateChargeProgres {nodes [
   UpdateChargeProgres() {is=title K2Node_FunctionEntry}
   n1 = SafeDivide(
    A=float - float(A=GetGameTimeInSeconds(), B=Get WindUpStartTime() -> (WindUpStartTime:double))
    B=Get WindUpDuration() -> (WindUpDuration:float)
   )
   > SetWindUpProgress(Value=@n1)
   > Branch(Condition="float >= float"(A=@n1, B=1))
   > SetWindUpState(IsActive=false) {is=title K2Node_CallFunction}
  ]}
]
variables [
 ConfigData {type=DataClass_RocketRam_C, default=/RocketRamGameplay/Items/RocketRam/DataAsset_RocketRam_DoNotModify}
 CurrentState {type=Enum_RocketRamState, default=WindUp}
 FortPawn {type=FortPawn, default=null}
 GA_ShowTrajectory {type=GameplayTag, default {tagName=Granted.Athena.RocketRam.Targeting}}
 GE_WindUpEnd {type=GameplayTag, default {tagName=Event.Athena.RocketRam.WindUpEnd}}
 GE_WindUpStart {type=GameplayTag, default {tagName=Event.Athena.RocketRam.WindUpStart}}
 Is Cooldown Active {type=boolean}
 IsTrajectoryShowing {type=boolean}
 IsWindUpActive {type=boolean}
 RocketRamWeapon {type=B_Weapon_RocketRam_C, default=null}
 Timer_UpdateBlockingUI {type=TimerHandle}
 WaitForShowTrajectoryAbilityAdded {type=AbilityAsync_WaitGameplayTagAdded, default=null}
 WaitForShowTrajectoryAbilityRemoved {type=AbilityAsync_WaitGameplayTagRemoved, default=null}
 WaitForWindUpEndEventHandle {type=AbilityAsync_WaitGameplayEvent, default=null}
 WaitForWindUpStartEventHandle {type=AbilityAsync_WaitGameplayEvent, default=null}
 WindUpDuration {type=number}
 WindUpStartTime {type=number}
 WindUpUpdateTimerHandle {type=TimerHandle}
]
widgets {
 info {
  inheritedWidgetCount=1
  parentClass=/Script/FortniteUI.FortWeaponReticleExtensionWidgetBase
  rootWidgetClass=&SizeBox
  widgetCount=6
 }
 widgets [
  SizeBox_0 {
   widgetClassPath=&SizeBox
   widget=SizeBox_0
   properties {bOverride_HeightOverride=true, bOverride_WidthOverride=true, heightOverride=128, widthOverride=128}
  }
  Overlay_1 {widgetClassPath=&Overlay, parent=SizeBox_0, widget=Overlay_1, slot {class=SizeBoxSlot}}
  Image_ChargeUp {
   widgetClassPath=&Image
   bIsVariable=true
   parent=Overlay_1
   widget=Image_ChargeUp
   properties {
    brush {
     bIsDynamicallyLoaded=false
     drawAs=Image
     imageType=NoImage
     mirroring=NoMirror
     resourceName=None
     resourceObject=/RocketRamGameplayUI/Materials/MI_UI_Progress_ChargeUp_RocketRam
     tiling=NoTile
     imageSize=<64 64>
     margin=<ltrb 0 0 0 0>
     outlineSettings {
      bUseBrushTransparency=false
      roundingType=HalfHeightRadius
      width=0
      color {colorUseRule=UseColor_Specified, specifiedColor=<rgba 0 0 0 0>}
      cornerRadii=<0 0 0 0>
     }
     tintColor {colorUseRule=UseColor_Specified, specifiedColor=<rgba 1 1 1 1>}
     uVRegion {bIsValid=false, max=<0 0>, min=<0 0>}
    }
    renderTransform {angle=6, scale=<1.2 1.2>, shear=<0 0>, translation=<7 6>}
   }
   slot {class=OverlaySlot, properties {horizontalAlignment=HAlign_Fill, verticalAlignment=VAlign_Fill}}
  }
  WBP_Reticle_CooldownTimer {
   widgetClassPath=/Game/Athena/HUD/reticles/Widgets/WBP_Reticle_CooldownTimer
   bIsVariable=true
   parent=Overlay_1
   widget=WBP_Reticle_CooldownTimer
   slot {class=OverlaySlot, properties {horizontalAlignment=HAlign_Center, verticalAlignment=VAlign_Center}}
  }
  Image_Crosshair {
   widgetClassPath=&Image
   bIsVariable=true
   parent=Overlay_1
   widget=Image_Crosshair
   properties {brush {
     bIsDynamicallyLoaded=false
     drawAs=Image
     imageType=NoImage
     mirroring=NoMirror
     resourceName=None
     resourceObject=/Game/Athena/HUD/reticles/reticle_cross_new
     tiling=NoTile
     imageSize=<16 16>
     margin=<ltrb 0 0 0 0>
     outlineSettings {
      bUseBrushTransparency=false
      roundingType=HalfHeightRadius
      width=0
      color {colorUseRule=UseColor_Specified, specifiedColor=<rgba 0 0 0 0>}
      cornerRadii=<0 0 0 0>
     }
     tintColor {colorUseRule=UseColor_Specified, specifiedColor=<rgba 1 1 1 1>}
     uVRegion {bIsValid=false, max=<0 0>, min=<0 0>}
    }}
   slot {class=OverlaySlot, properties {horizontalAlignment=HAlign_Center, verticalAlignment=VAlign_Center}}
  }
  TutorialHighlight {widgetClassPath=&TutorialHighlight}
 ]
}
